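# Physics asset for the M_MED_Safe backpack cosmetic, providing physics bodies for the safe's door and handle bones.
# Each of dyn_door and dyn_handle has a single box shape, with the door box offset and rotated from its bone.
# One constraint joins handle to door as a hinge, locking twist and swing 1 and leaving swing 2 free.
# No constraint in the asset ties the door to another body.
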PhysicsAsset /BRCosmetics/Accessories/FORT_Backpacks/M_MED_Safe/Meshes/M_MED_Safe_Backpack_Physics
bodies {
 dyn_door [{
   shapeType=Box
   extentX=23.6107
   extentY=6.6243
   extentZ=29.6084
   shapeName=dyn_door_box
   center=<-2.3825 9.941 -0.1557>
   rotation=<rot 3.7153 -89.9923 0.0439>
  }]
 dyn_handle [{shapeType=Box, extentX=7.5161, extentY=2.8631, extentZ=11.6303, shapeName=dyn_handle_box, center=<0 0 -5>}]
}
constraints [{
  bone1Name=dyn_handle
  bone2Name=dyn_door
  swing1LimitDegrees=90
  swing1Motion=Locked
  swing2LimitDegrees=90
  swing2Motion=Free
  twistLimitDegrees=15
  twistMotion=Locked
 }]
