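# Animation Blueprint for the M_MED_Safe backpack cosmetic that adds secondary physics motion to its skeleton.
# The input pose feeds an Anim Dynamics node that swings the dyn_door bone as an angular spring within a one-sided 90-degree hinge limit.
# A RigidBody node then simulates the backpack's override physics asset with custom gravity and an LOD threshold.
# The Event Graph holds only unconnected default nodes.
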AnimBlueprint /BRCosmetics/Accessories/FORT_Backpacks/M_MED_Safe/Meshes/M_MED_Safe_Backpack_AnimBP
target_skeleton=&/M_MED_Safe_Backpack_Skeleton
graphs [
 AnimGraph {nodes [
   AnimGraphNode_SubInput_0 = Input Pose()
   n1 = Anim Dynamics - Bone: dyn_door(ComponentPose=Local To Component(LocalPose=@AnimGraphNode_SubInput_0), Alpha=1, ExternalForce="0,0,0") {
    angularBiasOverride=0.99
    angularDampingOverride=0.9
    angularSpringConstant=150
    bAngularSpring=true
    simulationSpace=BoneRelative
    alphaBoolBlend {blendOption=Linear}
    boundBone {boneName=dyn_door}
    componentAppliedLinearAccClamp=<10000 10000 10000>
    componentLinearAccScale=<1 1 3>
    componentLinearVelScale=<1 1 3>
    physicsBodyDefinitions [{
      boundBone {boneName=dyn_door}
      boxExtents=<30 22 30>
      constraintSetup {angularLimitsMax=<0 0 0>, angularLimitsMin=<0 0 -90>, angularTarget=<0 0 0>}
      localJointOffset=<0 3 0>
     }]
   }
   n2 = RigidBody(ComponentPose=@n1) {
    bClampLinearTranslationLimitToRefPose=true
    bOverrideWorldGravity=true
    lODThreshold=1
    overridePhysicsAsset=&/M_MED_Safe_Backpack_Physics
    alphaBoolBlend {blendOption=Linear}
    componentAppliedLinearAccClamp=<5000 5000 5000>
    componentLinearAccScale=<1 1>
    componentLinearVelScale=<1 1>
    overrideWorldGravity {z=-980}
   }
   Output Pose(Result=Component To Local(ComponentPose=@n2)) {layerGroup=None}
  ]}
 EventGraph {nodes [
   TryGetPawnOwner()
   Event BlueprintUpdateAnimation()
  ]}
]
variables []
